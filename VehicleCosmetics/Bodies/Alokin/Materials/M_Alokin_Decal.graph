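# Masked, two-sided Default Lit decal material for the Alokin vehicle body cosmetic, enabled for skeletal meshes and mesh deformers.
# The inverted vertex color green channel drives every blend.
# As green falls, the decal base color texture fades to grey and the normal map fades to flat.
# The same mask blends two constants into an inverted scalar, whose destination material input is not shown.
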Material /VehicleCosmetics/Bodies/Alokin/Materials/M_Alokin_Decal
graph {
 properties {
  bAutomaticallySetUsageInEditor=false
  bUsedWithMeshDeformer=true
  bUsedWithSkeletalMesh=true
  blendMode=BLEND_Masked
  twoSided=true
  usedShadingModels=Default Lit
 }
 nodes [
  n1 = OneMinus(None=VertexColor().G)
  LinearInterpolate(A=TextureSample() {texture=/VehicleCosmetics/Bodies/Alokin/Textures/Chassis/T_Alokin_Decal_BC}.RGB, B=Constant() {r=0.5}, Alpha=@n1)
  LinearInterpolate(
   A=TextureSample() {samplerType=SAMPLERTYPE_Normal, texture=/VehicleCosmetics/Bodies/Alokin/Textures/Chassis/T_Alokin_Decal_N}.RGB
   B=Constant3Vector() {constant=<rgba 0.5 0.5 1 0>}
   Alpha=@n1
  )
  OneMinus(None=LinearInterpolate(Alpha=@n1) {constA=0.8, constB=0.2})
 ]
}
